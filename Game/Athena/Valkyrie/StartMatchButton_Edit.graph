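# StartMatchButton_Edit is a Widget Blueprint button labeled "Start Match", derived from CommonButtonLegacy.
# An Overlay root holds one centered CommonTextBlock styled as an XL header in the Burbank Black font.
# The PreConstruct, Construct and Tick events are empty, so the Blueprint adds only visuals and no click logic of its own.
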WidgetBlueprint /Game/Athena/Valkyrie/StartMatchButton_Edit
graphs [EventGraph {nodes [
   Event PreConstruct()
   Event Construct()
   Event Tick()
  ]}]
variables []
widgets {
 info {inheritedWidgetCount=1, parentClass=&CommonButtonLegacy, rootWidgetClass=&Overlay, widgetCount=3}
 widgets [
  Overlay_18 {widgetClassPath=&Overlay, widget=Overlay_18}
  CommonTextBlock_33 {
   widgetClassPath=&CommonTextBlock
   parent=Overlay_18
   widget=CommonTextBlock_33
   properties {
    justification=Center
    style=/FortUILibrary/Styles/Typography/Foundation/TextStyle-Header-XL.TextStyle-Header-XL_C
    text=Start Match
    font {
     bForceMonospaced=false
     bMaterialIsStencil=false
     fontMaterial=null
     fontObject=&Burbank
     letterSpacing=0
     monospacedWidth=1
     size=72
     skewAmount=0
     typefaceFontName=Burbank-Black
     outlineSettings {
      bApplyOutlineToDropShadows=false
      bMiteredCorners=false
      bSeparateFillAlpha=false
      outlineMaterial=null
      outlineSize=0
      outlineColor=<rgba 0 0 0 1>
     }
    }
    shadowOffset=<0 0>
   }
   slot {class=OverlaySlot, properties {horizontalAlignment=HAlign_Fill, verticalAlignment=VAlign_Center}}
  }
  InputActionWidget {widgetClassPath=&CommonActionWidget, bInherited=true, widget=null}
 ]
}
