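# ForcedIntroWidget_Edit is the Valkyrie forced intro screen shown while loading into an Edit Session, built on FortUIStateWidget_ForcedIntro_Valkyrie.
# An overlay fills with season key art and stacks Burbank loading text above a Start Match button in a bottom-aligned box.
# The button starts collapsed, and this Blueprint has no logic to reveal it.
# Its click event calls StartMatch on the widget, the only graph logic here.
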WidgetBlueprint /Game/Athena/Valkyrie/ForcedIntroWidget_Edit
graphs [EventGraph {nodes [
   "On Button Clicked (Button_StartMatch)"!() -> (OutputDelegate, then, Button:Common Button Legacy Object Reference) {is=title K2Node_ComponentBoundEvent}
   > StartMatch(self:Fort UIState Widget Forced Intro Object Reference)
  ]}]
variables []
widgets {
 info {
  inheritedWidgetCount=2
  parentClass=/Script/FortniteUI.FortUIStateWidget_ForcedIntro_Valkyrie
  rootWidgetClass=&Overlay
  widgetCount=7
 }
 widgets [
  Overlay_19 {widgetClassPath=&Overlay, widget=Overlay_19}
  BackgroundImage {
   widgetClassPath=&Image
   bIsVariable=true
   parent=Overlay_19
   widget=BackgroundImage
   properties {brush {
     bIsDynamicallyLoaded=false
     drawAs=Image
     imageType=NoImage
     mirroring=NoMirror
     resourceName=None
     resourceObject=/Game/UI/Foundation/Textures/LoadingScreens/Fort_Load_SeasonKeyArt
     tiling=NoTile
     imageSize=<2048 1024>
     margin=<ltrb 0 0 0 0>
     outlineSettings {
      bUseBrushTransparency=false
      roundingType=HalfHeightRadius
      width=0
      color {colorUseRule=UseColor_Specified, specifiedColor=<rgba 0 0 0 0>}
      cornerRadii=<0 0 0 0>
     }
     tintColor {colorUseRule=UseColor_Specified, specifiedColor=<rgba 1 1 1 1>}
     uVRegion {bIsValid=false, max=<0 0>, min=<0 0>}
    }}
   slot {class=OverlaySlot, properties {horizontalAlignment=HAlign_Fill, verticalAlignment=VAlign_Fill}}
  }
  VerticalBox_0 {
   widgetClassPath=&VerticalBox
   parent=Overlay_19
   widget=VerticalBox_0
   slot {class=OverlaySlot, properties {verticalAlignment=VAlign_Bottom, padding=<ltrb 155 0 50 90>}}
  }
  LoadingText {
   widgetClassPath=&CommonTextBlock
   bIsVariable=true
   parent=VerticalBox_0
   widget=LoadingText
   properties {
    justification=Center
    text=Loading into your Edit Session...
    font {
     bForceMonospaced=false
     bMaterialIsStencil=false
     fontMaterial=null
     fontObject=&Burbank
     letterSpacing=0
     monospacedWidth=1
     size=48
     skewAmount=0
     typefaceFontName=Burbank-PlayerPower
     outlineSettings {
      bApplyOutlineToDropShadows=false
      bMiteredCorners=false
      bSeparateFillAlpha=false
      outlineMaterial=null
      outlineSize=0
      outlineColor=<rgba 0 0 0 1>
     }
    }
    shadowColorAndOpacity=<rgba 0 0.431 0.4427 1>
    shadowOffset=<2.5 2.5>
   }
   slot {class=VerticalBoxSlot}
  }
  Button_StartMatch {
   widgetClassPath=&/StartMatchButton_Edit
   bIsVariable=true
   parent=VerticalBox_0
   widget=Button_StartMatch
   properties {
    onButtonBaseClicked=(&self:WidgetTree.Button_StartMatch.HandleOnButtonClicked)
    onButtonBaseDoubleClicked=(&self:WidgetTree.Button_StartMatch.HandleOnButtonDoubleClicked)
    onButtonBaseHovered=(&self:WidgetTree.Button_StartMatch.HandleOnButtonHovered)
    onButtonBaseUnhovered=(&self:WidgetTree.Button_StartMatch.HandleOnButtonUnhovered)
    onSelectedChangedBase=(&self:WidgetTree.Button_StartMatch.HandleOnSelectedChanged)
    visibility=Collapsed
   }
   slot {class=VerticalBoxSlot}
  }
  Text_ForcedIntroState {widgetClassPath=&CommonTextBlock, bInherited=true, widget=null}
  WidgetStack_PrimaryContent {widgetClassPath=/Script/CommonUI.CommonActivatableWidgetStack, bInherited=true, widget=null}
 ]
}
